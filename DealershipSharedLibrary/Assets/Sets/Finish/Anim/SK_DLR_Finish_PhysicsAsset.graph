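# Physics asset for SK_DLR_Finish, a skeletal mesh in the dealership shared library's Finish set, apparently an animated hand holding a flag.
# Two capsule bodies approximate the mesh, a large one on the flag grip joint and a smaller one on the hand joint.
# A single constraint links the two bodies with equal swing and twist limit angles.
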PhysicsAsset /DealershipSharedLibrary/Assets/Sets/Finish/Anim/SK_DLR_Finish_PhysicsAsset
bodies {
 flag_grip_jnt [{shapeType=Capsule, length=319.9301, radius=255.0385, center=<152.5766 0.2563 151.5501>, rotation=<rot -0.0558 -90 19.3408>}]
 hand_jnt [{shapeType=Capsule, length=97.6916, radius=95.7181, center=<87.7457 7.345 -1.3841>, rotation=<rot -8.4976 -90 -20.3691>}]
}
constraints [{bone1Name=flag_grip_jnt, bone2Name=hand_jnt, swing1LimitDegrees=45, swing2LimitDegrees=45, twistLimitDegrees=45}]
